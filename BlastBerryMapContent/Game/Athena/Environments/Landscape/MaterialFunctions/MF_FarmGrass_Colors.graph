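# Exposed material function for farm landscape materials that tints grass and rock colors using shared vector parameters.
# FarmGrass_Tint and FarmRock_Tint multiply the input diffuses for the Grass and Rock outputs, and are also output raw.
# Two separate alpha inputs lerp from rock tint to grass tint for BlendedColors and BlendedColors_Bias.
# The grass desaturation and power-shaped alpha blend toward untinted grass feed no output.
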MaterialFunction /BlastBerryMapContent/Game/Athena/Environments/Landscape/MaterialFunctions/MF_FarmGrass_Colors
graph {
 properties {bExposeToLibrary=true}
 nodes [
  n1 = FunctionInput() {inputName=Grass_Diffuse}
  Desaturation(None=@n1, Fraction=ScalarParameter() {defaultValue=0.5, parameterName=Farm_Grass_Desat})
  FunctionInput()
  n2 = ComponentMask(None=VectorParameter() {parameterName=FarmRock_Tint, defaultValue=<rgba 0.705 0.5871 0.416 0>}.RGB) {b=true, g=true, r=true}
  FunctionOutput(None=@n2) {outputName=RockTint, sortPriority=1}
  n3 = ComponentMask(None=VectorParameter() {parameterName=FarmGrass_Tint, defaultValue=<rgba 1.5953 1.022 0.5709 0>}.RGB) {b=true, g=true, r=true}
  FunctionOutput(None=@n3) {outputName=GrassTint, sortPriority=3}
  FunctionOutput(None=LinearInterpolate(A=@n2, B=@n3, Alpha=FunctionInput() {bUsePreviewValueAsDefault=true, inputName=ColorBlend_Alpha, inputType=FunctionInput_Scalar})) {outputName=BlendedColors}
  FunctionOutput(None=LinearInterpolate(A=@n2, B=@n3, Alpha=FunctionInput() {bUsePreviewValueAsDefault=true, inputName=ColorBlend_Alpha_Bias, inputType=FunctionInput_Scalar})) {outputName=BlendedColors_Bias}
  n4 = Multiply(A=@n1, B=@n3)
  FunctionOutput(None=@n4) {outputName=Grass, sortPriority=2}
  FunctionOutput(None=Multiply(A=FunctionInput() {inputName=Rock_Diffuse}, B=@n2)) {outputName=Rock}
  n5 = Constant()
  LinearInterpolate(
   A=@n4
   B=@n1
   Alpha=Power(
    Base=FunctionInput(Preview=@n5) {bUsePreviewValueAsDefault=true, inputName=Alpha, inputType=FunctionInput_Scalar}
    Exp=ScalarParameter() {defaultValue=1, parameterName=Farm_Alpha_Pow}
   )
  )
 ]
}
