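# Material function, exposed to the library, that tints swamp grass and rock diffuse colors for landscape materials.
# It multiplies the Grass_Diffuse and Rock_Diffuse inputs by RGB tints read from the shared MPC_Landscape parameter collection.
# It outputs both tinted colors plus the raw grass and rock tints.
# A grass desaturation node and two local tint vector parameters are present but feed no output.
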MaterialFunction /BlastBerryMapContent/Game/Athena/Environments/Landscape/MaterialFunctions/MF_SwampGrass_Colors
graph {
 properties {bExposeToLibrary=true}
 nodes [
  n1 = FunctionInput() {inputName=Grass_Diffuse}
  Desaturation(None=@n1, Fraction=ScalarParameter() {parameterName=Swamp_Grass_Desat})
  n2 = ComponentMask(None=CollectionParameter() {collection=/BlastBerryMapContent/Game/Athena/Environments/Landscape/MPC/MPC_Landscape, parameterName=SwampRockTint}) {b=true, g=true, r=true}
  FunctionOutput(None=@n2) {outputName=RockTint, sortPriority=1}
  n3 = ComponentMask(None=CollectionParameter() {collection=&MPC_Landscape, parameterName=SwampGrassTint}) {b=true, g=true, r=true}
  FunctionOutput(None=@n3) {outputName=GrassTint, sortPriority=3}
  FunctionOutput(None=Multiply(A=@n1, B=@n3)) {outputName=Grass, sortPriority=2}
  FunctionOutput(None=Multiply(A=FunctionInput() {inputName=Rock_Diffuse}, B=@n2)) {outputName=Rock}
  VectorParameter() {parameterName=SwampGrass_Tint, defaultValue=<rgba 0.5168 0.3932 0.0519 0>}
  VectorParameter() {parameterName=SwampRock_Tint, defaultValue=<rgba 0.16 0.14 0.136 0>}
 ]
}
