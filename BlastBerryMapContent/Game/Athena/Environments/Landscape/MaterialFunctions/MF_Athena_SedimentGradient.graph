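# Material function that produces a tinted sediment color gradient based on world height, exposed as Arid_Rock parameters for Athena landscape materials.
# World Z, shifted by an offset and divided by a range, serves as the UV for a gradient texture.
# A quality switch gives Low and Medium their own desaturation parameter.
# The desaturated color is brightened by adding constant mid-gray, then multiplied by a tint for output.
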MaterialFunction /BlastBerryMapContent/Game/Athena/Environments/Landscape/MaterialFunctions/MF_Athena_SedimentGradient
graph {nodes [
  n1 = ComponentMask(None=WorldPosition().XYZ) {b=true}
  n2 = ScalarParameter() {defaultValue=8000, group=Arid_Rock, parameterName=Gradient_Offset}
  n3 = TextureSampleParameter2D(UVs=Divide(A=Add(A=@n1, B=@n2), B=ScalarParameter() {defaultValue=4000, group=Arid_Rock, parameterName=GradientRange})) {
   group=Arid_Rock
   parameterName=GradientMap
   samplerSource=SSM_Wrap_WorldGroupSettings
   texture=/Game/Environments/Sets/Arid/Textures/GradientSediments__M
  }
  n4 = ScalarParameter() {group=Arid_Rock, parameterName=Gradient_Desaturation}
  n5 = ScalarParameter() {defaultValue=0.5, group=Arid_Rock, parameterName=Gradient_Desaturation_LowMed}
  n6 = Desaturation(None=@n3.RGB, Fraction=QualitySwitch(Default=@n4, Low=@n5, High=@n4, Medium=@n5))
  n7 = Constant3Vector() {constant=<rgba 0.5 0.5 0.5 0>}
  FunctionOutput(None=Multiply(A=Add(A=@n6, B=@n7), B=VectorParameter() {group=Arid_Rock, parameterName=GradientMap_Tint, defaultValue=<rgba 1 0.7407 0.6667 0>}.RGB))
 ]}
